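# Sound cue for electric spark crackle on Fort electric traps, using ceiling trap spark recordings.
# A Random node picks one of five equally weighted mono spark waves and passes it through a Modulator for variation.
# The cue raises pitch and volume, uses the Fort_Traps sound class, and overrides attenuation for short-range falloff.
# Its concurrency override blocks new instances at the limit; despite the Loop name, the graph has no looping node.
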SoundCue /Game/Sounds/Fort_Traps/Electric_Trap_Loop_Electricity
root=Modulator_0
nodes [
 {id=Modulator_0, children [Random_0]}
 {id=Random_0, children [WavePlayer_4,WavePlayer_0,WavePlayer_1,WavePlayer_3,WavePlayer_2], properties {weights [1,1,1,1,1]}}
 {id=WavePlayer_0, properties {soundWaveAssetPtr=/Game/Sounds/Fort_Traps/Electric_Trap/CeilingTrap_Electric_Spark_mono_007}}
 {id=WavePlayer_1, properties {soundWaveAssetPtr=/Game/Sounds/Fort_Traps/Electric_Trap/CeilingTrap_Electric_Spark_mono_005}}
 {id=WavePlayer_2, properties {soundWaveAssetPtr=/Game/Sounds/Fort_Traps/Electric_Trap/CeilingTrap_Electric_Spark_mono_004}}
 {id=WavePlayer_3, properties {soundWaveAssetPtr=/Game/Sounds/Fort_Traps/Electric_Trap/CeilingTrap_Electric_Spark_mono_003}}
 {id=WavePlayer_4, properties {soundWaveAssetPtr=/Game/Sounds/Fort_Traps/Electric_Trap/CeilingTrap_Electric_Spark_mono_002}}
]
properties {
 attenuationSettings=&Fort_Traps_Att
 bOverrideAttenuation=true
 bOverrideConcurrency=true
 maxDistance=600
 pitchMultiplier=1.5
 soundClassObject=&Fort_Traps
 volumeMultiplier=1.25
 attenuationOverrides {falloffDistance=600, attenuationShapeExtents {x=0}}
 concurrencyOverrides {resolutionRule=PreventNew}
}
